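; Reboot Van actor for the Figment environment that turns off its own interaction in solo playlists.
; On BeginPlay the server waits for the current playlist data, then stores its max team size in Max Team Size.
; That variable's RepNotify disables interact and highlight when team size is one or less, and otherwise only prints that interaction stays on.
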(event EventBeginPlay
  (|CallParentFunction)
  (Networking|SwitchHasAuthority
    (:Authority
      ; Get squad size to check playlist:
      ; Then disable reboot van's interaction if in Solos
      (bind (_gamestate _playlist) (Class|FortAsyncActionCurrentPlaylistReady|PlaylistDataReadyAsync)
        (:OnReady
          (Variables|Default|SetMax%20Team%20Size (Variables|/Script/FortniteGame.FortPlaylist|GetMaxTeamSize _playlist)))))))

(event Collision|EventActorBeginOverlap
  (params (OtherActor :type Actor%20Object%20Reference)))

(event EventTick
  (params (DeltaSeconds :type Float%20%28single-precision%29)))

(fn OnRep_MaxTeamSize
  ; Disable Interaction and Highlight on the Reboot Van if in Solo Playlist
  (if (<= (Variables|Default|GetMax%20Team%20Size) 1)
    (if (Variables|Interaction|GetbAllowInteract)
      (Variables|Interaction|SetbAllowInteract false)
      (Variables|Interaction|SetbAllowHighlight false))
    (else
      (Development|PrintString "Non solo playlist, keep interact ON"))))

(fn ConstructionScript
  (|CallParentFunction))
