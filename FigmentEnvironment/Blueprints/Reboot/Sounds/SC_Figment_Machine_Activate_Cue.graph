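# Sound cue for activating a Figment reboot machine, combining a card-insert sound with a longer activation sound.
# A Mixer plays both waves together, with the card insert at a reduced level beneath the activation.
# Each wave passes through its own Attenuation node, both using the shared Athena van close-activation attenuation asset.
# The cue is routed to the important in-game mission sound class.
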SoundCue /FigmentEnvironment/Blueprints/Reboot/Sounds/SC_Figment_Machine_Activate_Cue
root=Mixer_0
nodes [
 {id=Attenuation_0, children [WavePlayer_2], properties {attenuationSettings=/Game/Athena/Sounds/SC_Machine/Fort_Van_Activate_Close_Att}}
 {id=Attenuation_1, children [WavePlayer_4], properties {attenuationSettings=&Fort_Van_Activate_Close_Att}}
 {id=Mixer_0, children [Attenuation_1, Attenuation_0], properties {inputVolume [0.35,1]}}
 {id=WavePlayer_2, properties {soundWaveAssetPtr=&/SC_Figment_Machine_Activate_Longer}}
 {id=WavePlayer_4, properties {soundWaveAssetPtr=&/SC_Figment_Machine_Card_Insert}}
]
properties {
 maxDistance=2300
 soundClassObject=&Fort_Mission_Ingame_Important_PSM
 volumeMultiplier=1
 attenuationOverrides {bAttenuate=false, bEnableReverbSend=false, bSpatialize=false, falloffDistance=2000, attenuationShapeExtents {x=1500}}
}
