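# Sound cue for the looping ambient hum of the Figment Reboot machine.
# A single looping wave player plays the Athena SC_Machine ambient loop at reduced volume under the base ambient sound class.
# Overridden attenuation adds HRTF spatialization, distance falloff, listener focus that quiets the sound when out of focus, and traced occlusion with a low-pass filter.
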SoundCue /FigmentEnvironment/Blueprints/Reboot/Sounds/SC_Figment_Machine_Ambient_Loop_Cue
root=WavePlayer_0
nodes [{id=WavePlayer_0, properties {bLooping=true, soundWaveAssetPtr=/Game/Athena/Sounds/SC_Machine/SC_Machine_Ambient_Loop_01}}]
properties {
 bOverrideAttenuation=true
 maxDistance=3000
 soundClassObject=/Game/Sounds/FortSoundClassesAndModes/Fort_12_Base_Ambient
 volumeMultiplier=0.3
 attenuationOverrides {
  bEnableListenerFocus=true
  bEnableOcclusion=true
  falloffDistance=2000
  nonFocusVolumeAttenuation=0.2
  occlusionLowPassFilterFrequency=2000
  occlusionTraceChannel=ECC_GameTraceChannel6
  spatializationAlgorithm=SPATIALIZATION_HRTF
  attenuationShapeExtents {x=1000}
 }
}
