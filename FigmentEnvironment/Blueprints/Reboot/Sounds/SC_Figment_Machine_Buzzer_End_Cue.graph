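# Sound cue for the buzzer end sound of the Reboot machine in the Figment environment.
# A single wave player outputs the shared Athena SC_Machine_Charge_End wave at a heavily reduced volume, with no mixing or randomization.
# The cue enables its own attenuation override for falloff, though it also references a supply-drop machine attenuation asset.
# It routes to an important in-game mission sound class.
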SoundCue /FigmentEnvironment/Blueprints/Reboot/Sounds/SC_Figment_Machine_Buzzer_End_Cue
root=WavePlayer_3
nodes [{id=WavePlayer_3, properties {soundWaveAssetPtr=/Game/Athena/Sounds/SC_Machine/SC_Machine_Charge_End}}]
properties {
 attenuationSettings=/Game/Athena/Sounds/SupplyDrops/Fort_SC_Machine_Close_Athena_Att
 bOverrideAttenuation=true
 maxDistance=4000
 soundClassObject=&Fort_Mission_Ingame_Important_PSM
 volumeMultiplier=0.1
 attenuationOverrides {falloffDistance=2000, attenuationShapeExtents {x=2000}}
}
