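# Sound cue for the "not enough gold" feedback on the Figment reboot machine, reusing an Athena reboot sound wave.
# A single wave player feeds a modulator that varies volume slightly and keeps pitch fixed, at a reduced overall cue volume.
# It uses an important in-game mission sound class with attenuation overrides.
# Owner-limited concurrency allows one instance and prevents new plays while it is active.
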SoundCue /FigmentEnvironment/Blueprints/Reboot/Sounds/SC_Figment_Machine_Card_NotEnoughGold
root=Modulator_0
nodes [
 {id=Modulator_0, children [WavePlayer_1], properties {pitchMax=1, pitchMin=1, volumeMin=0.9}}
 {id=WavePlayer_1, properties {soundWaveAssetPtr=/Game/Athena/Sounds/SC_Machine/Enhanced_Reboot_NotEnoughGold}}
]
properties {
 maxDistance=2097152
 soundClassObject=&Fort_Mission_Ingame_Important_PSM
 volumeMultiplier=0.25
 attenuationOverrides {falloffDistance=1500, attenuationShapeExtents {x=1000}}
 concurrencyOverrides {bLimitToOwner=true, maxCount=1, resolutionRule=PreventNew}
}
