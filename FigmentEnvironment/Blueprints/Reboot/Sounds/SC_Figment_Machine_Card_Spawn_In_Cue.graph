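# Sound cue for the Figment reboot machine, named for its card spawn-in moment.
# A single wave player feeds the root directly, reusing the Athena SC_Machine "Power Ready" wave at reduced volume with no randomization or layering.
# Overridden attenuation turns off distance-based volume falloff, and the cue routes through the important in-game mission sound class.
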SoundCue /FigmentEnvironment/Blueprints/Reboot/Sounds/SC_Figment_Machine_Card_Spawn_In_Cue
root=WavePlayer_0
nodes [{id=WavePlayer_0, properties {soundWaveAssetPtr=/Game/Athena/Sounds/SC_Machine/SC_Machine_Power_Ready_01}}]
properties {
 bOverrideAttenuation=true
 maxDistance=2097152
 soundClassObject=&Fort_Mission_Ingame_Important_PSM
 volumeMultiplier=0.3
 attenuationOverrides {bAttenuate=false, falloffDistance=10000000, attenuationShapeExtents {x=1500}}
}
