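# Sound cue for the Figment reboot machine's cooldown, playing a looping alert over a quieter looping hum.
# A Mixer blends both wave players, with only the alert passing through an Enveloper whose indefinitely repeating five-second volume curve swells and fades it.
# The cue overrides attenuation with logarithmic distance falloff and occlusion low-pass filtering, and routes to the important in-game mission sound class at reduced volume.
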SoundCue /FigmentEnvironment/Blueprints/Reboot/Sounds/SC_Figment_Machine_Cooldown_Cue
root=Mixer_0
nodes [
 {
  id=Enveloper_0
  children [WavePlayer_4]
  properties {
   bLoop=true
   bLoopIndefinitely=true
   loopEnd=5
   pitchCurve {editorCurveData {keys [{time=0}]}}
   volumeCurve {editorCurveData {keys [<at 0 0>,<at 1 1>,<at 3.1549 0.6379>,<at 5 0>]}}
  }
 }
 {id=Mixer_0, children [Enveloper_0,WavePlayer_5], properties {inputVolume [1,0.3]}}
 {id=WavePlayer_4, properties {bLooping=true, soundWaveAssetPtr=/Game/Athena/Sounds/SC_Machine/SC_Machine_CoolDown_Alert_Loop}}
 {id=WavePlayer_5, properties {bLooping=true, soundWaveAssetPtr=/Game/Athena/Sounds/SC_Machine/SC_Machine_CoolDown_Hum_Loop}}
]
properties {
 bOverrideAttenuation=true
 maxDistance=1412
 soundClassObject=&Fort_Mission_Ingame_Important_PSM
 volumeMultiplier=0.3
 attenuationOverrides {
  bEnableOcclusion=true
  distanceAlgorithm=Logarithmic
  falloffDistance=900
  occlusionInterpolationTime=1
  occlusionLowPassFilterFrequency=5000
  occlusionTraceChannel=ECC_GameTraceChannel11
  attenuationShapeExtents {x=512}
 }
}
