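# Sound cue for the horn alert of the Figment Reboot machine, built to carry both near and far.
# A mixer plays two layers at equal volume: the close horn wave and a separate distant horn wave.
# Each layer passes through its own attenuation node, using a close machine preset and a distant rift preset respectively, so distance shapes the blend.
# The cue routes to an important in-game mission sound class.
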SoundCue /FigmentEnvironment/Blueprints/Reboot/Sounds/SC_Figment_Machine_Horn_Alert_Cue
root=Mixer_0
nodes [
 {id=Attenuation_0, children [WavePlayer_1], properties {attenuationSettings=/Game/Athena/Sounds/SC_Machine/Van_Redux_Rift_Distant_Att}}
 {id=Attenuation_1, children [WavePlayer_3], properties {attenuationSettings=/Game/Athena/Sounds/SupplyDrops/Fort_SC_Machine_Close_Athena_Att}}
 {id=Mixer_0, children [Attenuation_1, Attenuation_0], properties {inputVolume [1,1]}}
 {id=WavePlayer_1, properties {soundWaveAssetPtr=&/SC_Machine_Horn_Alert_Distant}}
 {id=WavePlayer_3, properties {soundWaveAssetPtr=&/SC_Machine_Horn_Alert}}
]
properties {
 maxDistance=40000
 soundClassObject=&Fort_Mission_Ingame_Important_PSM
 attenuationOverrides {falloffDistance=2500, attenuationShapeExtents {x=1500}}
 concurrencyOverrides {maxCount=1}
}
