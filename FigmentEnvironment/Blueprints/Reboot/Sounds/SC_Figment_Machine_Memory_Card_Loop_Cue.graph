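# Looping ambient sound cue for the Reboot memory card machine in the Figment environment.
# A looping wave player feeds the SC_Machine memory card loop from Athena into an Enveloper, whose volume curve drops from full to half over ten seconds.
# The cue plays on the base ambient sound class with its own attenuation override, adding distance falloff and low-pass occlusion on a game trace channel.
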SoundCue /FigmentEnvironment/Blueprints/Reboot/Sounds/SC_Figment_Machine_Memory_Card_Loop_Cue
root=Enveloper_0
nodes [
 {
  id=Enveloper_0
  children [WavePlayer_0]
  properties {pitchCurve {editorCurveData {keys [{time=0}]}}, volumeCurve {editorCurveData {keys [<at 0 1>,<at 10 0.5>]}}}
 }
 {id=WavePlayer_0, properties {bLooping=true, soundWaveAssetPtr=/Game/Athena/Sounds/SC_Machine/SC_Machine_Memory_Card_Loop}}
]
properties {
 bOverrideAttenuation=true
 maxDistance=3000
 soundClassObject=/Game/Sounds/FortSoundClassesAndModes/Fort_12_Base_Ambient
 volumeMultiplier=0.728
 attenuationOverrides {
  bEnableOcclusion=true
  falloffDistance=2000
  occlusionInterpolationTime=0.3
  occlusionLowPassFilterFrequency=4000
  occlusionTraceChannel=ECC_GameTraceChannel11
  attenuationShapeExtents {x=1000}
 }
}
